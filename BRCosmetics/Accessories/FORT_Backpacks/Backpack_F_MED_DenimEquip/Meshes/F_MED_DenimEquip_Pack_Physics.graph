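# Physics asset for the DenimEquip medium female backpack cosmetic, giving collision and joints to the pack and its two dynamic straps.
# The pack body is a compound of five capsules, while each left and right strap is a chain of four single-capsule bodies.
# Each strap root locks swing against the pack, and the joints below carry swing angles that widen down the chain with very narrow twist angles.
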PhysicsAsset /BRCosmetics/Accessories/FORT_Backpacks/Backpack_F_MED_DenimEquip/Meshes/F_MED_DenimEquip_Pack_Physics
bodies {
 dyn_strap_1_l [{shapeType=Capsule, length=2, radius=1.2, shapeName=dyn_strap_1_l_capsule, center=<2 0 0>, rotation=<rot -90 0 0>}]
 dyn_strap_1_r [{shapeType=Capsule, length=2, radius=1.2, shapeName=dyn_strap_1_r_capsule, center=<2 0 0>, rotation=<rot -90 0 0>}]
 dyn_strap_2_l [{shapeType=Capsule, length=2, radius=2, shapeName=dyn_strap_2_l_capsule, center=<2 0 0>, rotation=<rot -90 0 0>}]
 dyn_strap_2_r [{shapeType=Capsule, length=2, radius=2, shapeName=dyn_strap_2_r_capsule, center=<2 0 0>, rotation=<rot -90 0 0>}]
 dyn_strap_3_l [{shapeType=Capsule, length=2, radius=2, shapeName=dyn_strap_3_l_capsule, center=<2 0 0>, rotation=<rot -90 0 0>}]
 dyn_strap_3_r [{shapeType=Capsule, length=2, radius=2, shapeName=dyn_strap_3_r_capsule, center=<2 0 0>, rotation=<rot -90 0 0>}]
 dyn_strap_l [{shapeType=Capsule, length=2, radius=1.5, shapeName=dyn_strap_l_capsule, center=<2 0 0>, rotation=<rot -90 0 0>}]
 dyn_strap_r [{shapeType=Capsule, length=2, radius=1.5, shapeName=dyn_strap_r_capsule, center=<2 0 0>, rotation=<rot -90 0 0>}]
 pack [
  {shapeType=Capsule, length=4, radius=8, shapeName=pack_capsule}
  {
   shapeType=Capsule
   length=32.2194
   radius=10.4784
   shapeName=backpack_root_capsule
   center=<0 11.6976 -1.0348>
   rotation=<rot 13.6096 -90 90>
  }
  {shapeType=Capsule, length=8.2786, radius=11, shapeName=backpack_root_capsule, center=<0 12.5763 -18.8355>, rotation=<rot 13.6098 -90 90>}
  {shapeType=Capsule, length=4, radius=7, shapeName=pack_capsule, center=<0 -6.7645 -4.4484>, rotation=<rot 0 0 -90>}
  {
   shapeType=Capsule
   length=20.2786
   radius=10.4784
   shapeName=backpack_root_capsule
   center=<0 11.9176 -10.8907>
   rotation=<rot 13.6098 -90 90>
  }
 ]
}
constraints [
 {
  bone1Name=dyn_strap_l
  bone2Name=pack
  swing1LimitDegrees=25
  swing1Motion=Locked
  swing2LimitDegrees=25
  swing2Motion=Locked
  twistLimitDegrees=15
 }
 {bone1Name=dyn_strap_1_l, bone2Name=dyn_strap_l, swing1LimitDegrees=15, swing2LimitDegrees=15, twistLimitDegrees=1}
 {bone1Name=dyn_strap_2_l, bone2Name=dyn_strap_1_l, swing1LimitDegrees=20, swing2LimitDegrees=25, twistLimitDegrees=1}
 {bone1Name=dyn_strap_3_l, bone2Name=dyn_strap_2_l, swing1LimitDegrees=25, swing2LimitDegrees=25, twistLimitDegrees=1}
 {
  bone1Name=dyn_strap_r
  bone2Name=pack
  swing1LimitDegrees=25
  swing1Motion=Locked
  swing2LimitDegrees=25
  swing2Motion=Locked
  twistLimitDegrees=15
 }
 {bone1Name=dyn_strap_1_r, bone2Name=dyn_strap_r, swing1LimitDegrees=15, swing2LimitDegrees=15, twistLimitDegrees=1}
 {bone1Name=dyn_strap_2_r, bone2Name=dyn_strap_1_r, swing1LimitDegrees=20, swing2LimitDegrees=25, twistLimitDegrees=1}
 {bone1Name=dyn_strap_3_r, bone2Name=dyn_strap_2_r, swing1LimitDegrees=25, swing2LimitDegrees=25, twistLimitDegrees=1}
]
